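# Translucent, two-sided lit particle material, named for the inward wisp of the BluGlo Pylon mission, that draws a scrolling, soft-ended streak.
# Emissive adds panned noise to particle color, scaled by a dynamic parameter that also drives tiling and pan speed.
# Opacity samples a beam texture using only scrolling V, masked by a V-end gradient, optional fresnel and depth fade, then particle alpha.
# An optional switch scales opacity by time of day.
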Material /Game/Missions/Secondary/BluGloPylon/Assets/Materials/M_BlugloPylon_WispIn_01
graph {
 properties {
  bEnableResponsiveAA=true
  bUsedWithMeshParticles=true
  bUsedWithNiagaraMeshParticles=true
  bUsedWithParticleSprites=true
  blendMode=BLEND_Translucent
  pixelDepthOffsetMode=PDOM_Legacy
  translucencyLightingMode=TLM_VolumetricPerVertexDirectional
  twoSided=true
  usedShadingModels=Default Lit
 }
 nodes [
  n1 = Multiply(
   A=DynamicParameter() {defaultValue=<rgba 0.1 1 1 1>, paramNames [Emissive, Tiling, Panner Speed, Param4]}.Panner Speed
   B=Time() {bOverride_Period=true, period=1}
  )
  n2 = Panner(Coordinate=TextureCoordinate() {vTiling=0.2}, Time=@n1) {speedY=1}
  n3 = ParticleColor()
  n4 = Multiply(
   A=Multiply(
    A=TextureSample(UVs=@n2) {texture=&T_NoisePack_01}.R
    B=ScalarParameter() {defaultValue=1, parameterName=Noise Boost}
   )
   B=@n3.RGB
  )
  MaterialFunctionCall(Emissive=Multiply(A=Add(A=@n4, B=@n3.RGB), B=DynamicParameter() {defaultValue=<rgba 0.1 1 1 1>, paramNames [Emissive, Tiling, Panner Speed, Param4]}.Emissive)) {materialFunction=&MF_FX_EmissiveAdjustments}
  n5 = TextureCoordinate()
  n6 = AppendVector(A=Constant() {r=1}, B=DynamicParameter() {defaultValue=<rgba 0.1 1 1 1>, paramNames [Emissive, Tiling, Panner Speed, Param4]}.Tiling)
  n7 = MaterialFunctionCall(Float2=Panner(Coordinate=Multiply(A=@n5, B=@n6), Time=@n1) {bFractionalPart=true, speedY=1}) {materialFunction=&BreakOutFloat2Components}
  n8 = TextureSample(UVs=MaterialFunctionCall(X=@n7.G, Y=@n7.G) {materialFunction=&MakeFloat2}.Result) {samplerType=SAMPLERTYPE_LinearColor, texture=/Game/Effects/Fort_Effects/Textures/Tile/T_Raygun_Beam_01}
  n9 = MaterialFunctionCall(Float2=TextureCoordinate()) {materialFunction=&BreakOutFloat2Components}
  n10 = OneMinus(None=@n9.G)
  n11 = Clamp(None=Multiply(A=Multiply(A=@n10, B=@n9.G), B=ScalarParameter() {defaultValue=6, parameterName=Gradient Scale}) {constB=6})
  n12 = ScalarParameter() {defaultValue=1.1, parameterName=Fresnel Brightness}
  n13 = ScalarParameter() {defaultValue=2.5, parameterName=Fresnel Exponent}
  n14 = StaticSwitchParameter(
   True=Multiply(
    A=@n11
    B=MaterialFunctionCall(Brightness=@n12, Exponent=@n13) {materialFunction=&MF_Fresnel_FX}.Fresnel Double Sided
   )
   False=@n11
  ) {defaultValue=true, parameterName=Use Fresnel}
  n15 = Multiply(A=@n3.A, B=DepthFade(Opacity=Multiply(A=@n8.R, B=@n14), FadeDistance=ScalarParameter() {defaultValue=15, parameterName=Fade Distance}))
  n16 = ScalarParameter() {defaultValue=1, parameterName=OpacityMorning}
  n17 = ScalarParameter() {defaultValue=1, parameterName=OpacityDay}
  n18 = ScalarParameter() {defaultValue=1, parameterName=OpacityEvening}
  n19 = ScalarParameter() {defaultValue=1, parameterName=OpacityNight}
  StaticSwitchParameter(
   True=Multiply(
    A=@n15
    B=MaterialFunctionCall(Color_Morning=@n16, Color_Day=@n17, Color_Evening=@n18, Color_Night=@n19) {materialFunction=&Color_Time_of_Day}.Result
   )
   False=@n15
  ) {parameterName=Use_color_time_of_day}
 ]
}
